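# Niagara system for the GrowMallet pickaxe's trail, a single ribbon emitter that follows pairs of weapon bones or sockets.
# User.SwingSectionIndex selects the socket pair, and shared anim-trail modules spawn and update short-lived, locked particles between them.
# Ribbons narrow and fade over their life, with color and alpha scaled down by time of day.
# An emitter color curve defines the tint, and dynamic material parameters include a value driven by looping emitter age.
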NiagaraSystem /BRCosmetics/Weapons/FORT_Melee/Pickaxe_GrowMallet/FX/NS_Pickaxe_GrowMallet_Trail
dependencies {
 usedDataInterfaces [&NiagaraDataInterfaceColorCurve, &NiagaraDataInterfaceCurve, &NiagaraDataInterfaceSocketReader]
 usedDynamicInputs []
 usedModules [
  &SystemState
  &self:SystemSpawnScript.SystemScriptSource.SystemScriptGraph.NiagaraNodeAssignment_1.SetVariables_#1
 ]
}
emitters {NE_Trail_01 {
  emitterSpawnScript [&FN_RibbonIDOffsetUtility()]
  emitterUpdateScript [
   &EmitterState()
   &FN_SpawnAnimTrail(Spawn Interval=0)
   /SetVariables(Emitter.Curve for Colors=DataInterface() {propertyValues {
      bExposeCurve=true
      AlphaCurve {keys [<at 0.15 0>,<at 0.2 10>,<at 0.21 15>,<at 0.45 1>,<at 0.5 5>,<at 0.64 2.5>,<at 0.65 15>,<at 0.7 0>]}
      BlueCurve {keys [<at 0 0>,<at 0.2 0>,<at 0.21 0.0499>,<at 0.325 0.0998>,<at 0.45 0.0899>,<at 0.5 0.0858>,<at 0.64 0.0296>,<at 0.65 0>,<at 1 0>]}
      GreenCurve {keys [<at 0 0>,<at 0.2 0>,<at 0.21 0.0499>,<at 0.325 0.0998>,<at 0.45 0.0899>,<at 0.5 0.104>,<at 0.64 0.0513>,<at 0.65 0>,<at 1 0>]}
      RedCurve {keys [<at 0 0>,<at 0.2 0>,<at 0.21 0.1739>,<at 0.325 0.3478>,<at 0.45 0.3652>,<at 0.5 0.3813>,<at 0.64 0.3813>,<at 0.65 0>,<at 1 0>]}
     }}) {isSetParameters=true}
  ]
  particleSpawnScript [
   &FN_InitializeAnimTrailBoneIndecies(
    First Bone or Socket Index:NiagaraInt32=&MergeHelper_IntSwitch(0=2, 1=0, 2=2, 3=4, 4=?, 5=?, Index=$User.SwingSectionIndex, Num=::NewEnumerator2)
    Second Bone or Socket Index:NiagaraInt32=&MergeHelper_IntSwitch(0=3, 1=1, 2=3, 3=5, 4=?, 5=?, Index=$User.SwingSectionIndex, Num=::NewEnumerator2)
   )
   &FN_InitializeAnimTrail(Color=<rgba 1 1 1 3>, Lifetime=0.175)
  ]
  particleUpdateScript [
   &SolveForcesAndVelocity(Clamp Velocity=false, Limit Acceleration=false)
   &FN_UpdateAnimTrail(Lock Particles=true) {module=FN_UpdateAnimTrail001}
   &DynamicMaterialParameters(
    Index 0 Param 1=1
    Index 1 Param 3=&FloatFromCurve(
     FloatCurve=DataInterface() {propertyValues {Curve {keys [<at 0 0>, {arriveTangent=1.9814, tangentMode=RCTM_User, time=1, value=1}]}}}
     CurveIndex=&ModuloFloat(A=$Emitter.Age, B=1)
     Scale Curve=0
    )
    Index 1 Param 4=1
    Index 2 Param 1=10
    Index 2 Param 2=1
    Index 2 Param 3=0.5
    Index 2 Param 4=0.6
    Module.Index0Float0Write=true
    Module.Index0Float1Write=true
    Module.Index0Float2Write=true
    Module.Index0Float3Write=true
    Module.Index1Float0Write=true
    Module.Index1Float1Write=true
    Module.Index1Float2Write=true
    Module.Index1Float3Write=true
    Module.Index2Float0Write=true
    Module.Index2Float1Write=true
    Module.Index2Float2Write=true
    Module.Index2Float3Write=true
    Write Parameter Index 1=::NewEnumerator1
    Write Parameter Index 2=::NewEnumerator1
   )
   &ScaleColor(
    Scale Alpha=&FloatFromCurve(
     FloatCurve=DataInterface() {propertyValues {Curve {keys [{interpMode=RCIM_Cubic, tangentMode=RCTM_User, time=0, value=1}, {arriveTangent=-6.8776, tangentMode=RCTM_User, time=0.325, value=0}]}}}
     Scale Curve=&FNTimeOfDayScalar(Night Scale=0.5, Morning Scale=0.75, Evening Scale=0.75)
    )
    Scale RGB=&VectorFromFloat(Value=&FloatFromCurve(
      FloatCurve=DataInterface() {propertyValues {Curve {keys [<at 0 1>,<at 1 1>]}}}
      Scale Curve=&FNTimeOfDayScalar(Night Scale=0.5, Morning Scale=0.75, Evening Scale=0.75)
     ))
    ScaleA=true
    ScaleRGB=true
   ) {module=ScaleColor001}
   &ScaleRibbonWidth(Ribbon Width Scale=&FloatFromCurve(FloatCurve=DataInterface() {propertyValues {Curve {keys [<at 0 1>,<at 1 0>]}}}))
  ]
  rendererClasses [&NiagaraRibbonRendererProperties]
  renderers [{rendererClass=&NiagaraRibbonRendererProperties}]
 }}
summary {
 systemName=NS_Pickaxe_GrowMallet_Trail
 emitters [{emitterName=NE_Trail_01, rendererClasses [&NiagaraRibbonRendererProperties]}]
 userVariables [
  User.bIsFrontEnd {type=NiagaraBool}
  User.bIsFrontEndPreview {type=NiagaraBool}
  User.PhysicalSurface {type=NiagaraInt32}
  User.RibbonIDOffset {type=NiagaraInt32}
  User.SwingSectionIndex {type=NiagaraInt32}
 ]
}
