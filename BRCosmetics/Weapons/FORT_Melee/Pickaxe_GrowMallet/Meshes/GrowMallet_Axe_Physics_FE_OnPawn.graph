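# Physics asset for the GrowMallet pickaxe in its on-pawn FE variant, simulating a bag against the character's legs.
# A three-segment dyn_bag capsule chain hangs from a zero-size base_bag body, with locked-to-zero twist limits and small swing limits that tighten toward the tip.
# Thigh and calf capsules on both legs provide body shapes for the bag, each calf linked to its thigh by a constraint with free swing and twist.
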PhysicsAsset /BRCosmetics/Weapons/FORT_Melee/Pickaxe_GrowMallet/Meshes/GrowMallet_Axe_Physics_FE_OnPawn
bodies {
 base_bag [{shapeType=Capsule, length=0, radius=0, shapeName=base_bag_capsule}]
 calf_l [{shapeType=Capsule, length=30, radius=8, shapeName=calf_l_capsule, center=<-26 0 -0.2974>, rotation=<rot 90 180 0>}]
 calf_r [{shapeType=Capsule, length=30, radius=9.5, shapeName=calf_r_capsule, center=<26 0 -1.3537>, rotation=<rot 90 0 0>}]
 dyn_bag_01 [{shapeType=Capsule, length=12, radius=4, shapeName=dyn_bag_01_capsule, center=<1 0 0>, rotation=<rot 0 0 90>}]
 dyn_bag_02 [{shapeType=Capsule, length=12, radius=7, shapeName=dyn_bag_02_capsule, rotation=<rot 0 0 90>}]
 dyn_bag_03 [
  {shapeType=Capsule, length=13, radius=7, shapeName=dyn_bag_03_capsule, center=<1 1 3>, rotation=<rot 0 0 90>}
  {shapeType=Capsule, length=15, radius=6.5, shapeName=dyn_bag_03_capsule, center=<14 2 4>, rotation=<rot 0 0 90>}
 ]
 thigh_l [{shapeType=Capsule, length=34.6616, radius=9.5479, shapeName=thigh_l_capsule, center=<-22.1907 -0.6066 0.3009>, rotation=<rot 90 180 0>}]
 thigh_r [{shapeType=Capsule, length=34.6616, radius=9.5479, shapeName=thigh_r_capsule, center=<22.1907 0.6066 -0.3009>, rotation=<rot 90 180 0>}]
}
constraints [
 {bone1Name=dyn_bag_01, bone2Name=base_bag, swing1LimitDegrees=10, swing2LimitDegrees=10, twistLimitDegrees=0}
 {bone1Name=dyn_bag_02, bone2Name=dyn_bag_01, swing1LimitDegrees=5, swing2LimitDegrees=10, twistLimitDegrees=0}
 {bone1Name=dyn_bag_03, bone2Name=dyn_bag_02, swing1LimitDegrees=2, swing2LimitDegrees=5, twistLimitDegrees=0}
 {
  bone1Name=calf_l
  bone2Name=thigh_l
  swing1LimitDegrees=45
  swing1Motion=Free
  swing2LimitDegrees=45
  swing2Motion=Free
  twistLimitDegrees=45
  twistMotion=Free
 }
 {
  bone1Name=calf_r
  bone2Name=thigh_r
  swing1LimitDegrees=45
  swing1Motion=Free
  swing2LimitDegrees=45
  swing2Motion=Free
  twistLimitDegrees=45
  twistMotion=Free
 }
]
