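# Animation Blueprint for the GrowMallet pickaxe, adding secondary physics motion to its bag, flaps and brick.
# Each update copies body root, pelvis and leg socket transforms from the parent mesh, sets state flags, and picks front-end or in-game sim-space and gravity.
# Body-bone and handle adjustments feed three RigidBody nodes with separate physics assets, then flap and brick constraints.
# The cached result passes through chained AnimDynamics nodes and a flap nudge to output.
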AnimBlueprint /BRCosmetics/Weapons/FORT_Melee/Pickaxe_GrowMallet/Meshes/GrowMallet_Axe_AnimBP
target_skeleton=&/GrowMallet_Axe_Skeleton
graphs [
 AnimGraph {nodes [
   n1 = Anim Dynamics - Bone: dyn_flap_01_jnt(ComponentPose=Local To Component(LocalPose=Use cached pose 'Sim Cache'()), Alpha=1, ExternalForce="0,0,0") {
    angularDampingOverride=1
    angularSpringConstant=15000
    bAngularSpring=true
    bOverrideAngularDamping=true
    bUseGravityOverride=true
    gravityScale=0.35
    simulationSpace=RootRelative
    alphaBoolBlend {blendOption=Linear}
    boundBone {boneName=dyn_flap_01_jnt}
    componentAppliedLinearAccClamp=<3000 3000 3000>
    componentLinearAccScale=<0.5 0.5 0.5>
    componentLinearVelScale=<0.5 0.5 0.5>
    gravityOverride {z=-980}
    physicsBodyDefinitions [{
      boundBone {boneName=dyn_flap_01_jnt}
      boxExtents=<30 40 45>
      constraintSetup {angularLimitsMax=<0 22 3>, angularLimitsMin=<0 4 -10>, angularTarget=<1 0 0>}
      localJointOffset=<4 0 0>
     }]
    relativeSpaceBone {boneName=dyn_bag_03}
   }
   n2 = Anim Dynamics - Bone: dyn_flap_01_jnt!(ComponentPose=@n1, bAlphaBoolEnabled:bool=false, ExternalForce="0,0,0") -> (Pose) {
    alphaInputType=Bool
    angularDampingOverride=0.94
    angularSpringConstant=15000
    bAngularSpring=true
    bOverrideAngularDamping=true
    bUseGravityOverride=true
    gravityScale=0.35
    simulationSpace=BoneRelative
    alphaBoolBlend {blendOption=Linear}
    boundBone {boneName=dyn_flap_01_jnt}
    componentAppliedLinearAccClamp=<3000 3000 3000>
    componentLinearAccScale=<0.5 0.5 0.5>
    componentLinearVelScale=<0.5 0.5 0.5>
    gravityOverride {z=-1600}
    physicsBodyDefinitions [{
      boundBone {boneName=dyn_flap_01_jnt}
      boxExtents=<30 40 45>
      constraintSetup {angularLimitsMax=<0 5 3>, angularLimitsMin=<0 0 -5>, angularTarget=<1 0 0>}
      localJointOffset=<4 0 0>
     }]
    relativeSpaceBone {boneName=dyn_bag_03}
   }
   n3 = Anim Dynamics - Bone: dyn_flap_02_jnt(ComponentPose=@n2, Alpha=1, ExternalForce="0,0,0") {
    angularDampingOverride=1
    angularSpringConstant=15000
    bAngularSpring=true
    bOverrideAngularDamping=true
    bUseGravityOverride=true
    gravityScale=0.35
    simulationSpace=RootRelative
    alphaBoolBlend {blendOption=Linear}
    boundBone {boneName=dyn_flap_02_jnt}
    componentAppliedLinearAccClamp=<3000 3000 3000>
    componentLinearAccScale=<0.5 0.5 0.5>
    componentLinearVelScale=<0.5 0.5 0.5>
    gravityOverride {z=-980}
    physicsBodyDefinitions [{
      boundBone {boneName=dyn_flap_02_jnt}
      boxExtents=<30 40 45>
      constraintSetup {angularLimitsMax=<0 22 0>, angularLimitsMin=<0 -3 0>, angularTarget=<1 0 0>}
      localJointOffset=<4 0 0>
     }]
    relativeSpaceBone {boneName=dyn_bag_03}
   }
   n4 = Anim Dynamics - Bone: dyn_brick(ComponentPose=@n3, Alpha=1, ExternalForce="0,0,0") {
    angularDampingOverride=0.94
    angularSpringConstant=10000
    bAngularSpring=true
    bOverrideAngularDamping=true
    bUseGravityOverride=true
    gravityScale=0.35
    simulationSpace=RootRelative
    alphaBoolBlend {blendOption=Linear}
    boundBone {boneName=dyn_brick}
    componentAppliedLinearAccClamp=<3000 3000 3000>
    componentLinearAccScale=<0.5 0.5 0.5>
    componentLinearVelScale=<0.5 0.5 0.5>
    gravityOverride {z=-1800}
    physicsBodyDefinitions [{
      boundBone {boneName=dyn_brick}
      boxExtents=<40 50 55>
      constraintSetup {angularLimitsMax=<0 4 3>, angularLimitsMin=<0 0 -4>, angularTarget=<1 0 0>}
      localJointOffset=<2 0 0>
     }]
    relativeSpaceBone {boneName=dyn_bag_03}
   }
   n5 = Anim Dynamics - Bone: dyn_brick!(ComponentPose=@n4, bAlphaBoolEnabled:bool=True, ExternalForce="0,0,0") -> (Pose) {
    alphaInputType=Bool
    angularDampingOverride=1
    angularSpringConstant=15000
    bAngularSpring=true
    bOverrideAngularDamping=true
    bUseGravityOverride=true
    gravityScale=0.35
    simulationSpace=RootRelative
    alphaBoolBlend {blendOption=Linear}
    boundBone {boneName=dyn_brick}
    componentAppliedLinearAccClamp=<3000 3000 3000>
    componentLinearAccScale=<0.5 0.5 0.5>
    componentLinearVelScale=<0.5 0.5 0.5>
    gravityOverride {z=-980}
    physicsBodyDefinitions [{
      boundBone {boneName=dyn_brick}
      boxExtents=<30 40 45>
      constraintSetup {angularLimitsMax=<0 8 4>, angularLimitsMin=<0 0 -10>, angularTarget=<1 0 0>}
      localJointOffset=<4 0 0>
     }]
    relativeSpaceBone {boneName=dyn_bag_01}
   }
   n6 = Anim Dynamics - Bone: dyn_bag_01!(ComponentPose=@n5, bAlphaBoolEnabled:bool=True, ExternalForce="0,0,0") -> (Pose) {
    alphaInputType=Bool
    angularDampingOverride=0.9999
    angularSpringConstant=15000
    bAngularSpring=true
    bOverrideAngularDamping=true
    bUseGravityOverride=true
    bUsePlanarLimit=false
    gravityScale=0.35
    simulationSpace=RootRelative
    alphaBoolBlend {blendOption=Linear}
    boundBone {boneName=dyn_bag_01}
    componentAppliedLinearAccClamp=<3000 3000 3000>
    componentLinearAccScale=<0.5 0.5 0.5>
    componentLinearVelScale=<0.5 0.5 0.5>
    gravityOverride {z=-980}
    physicsBodyDefinitions [{
      boundBone {boneName=dyn_bag_01}
      boxExtents=<25 35 35>
      constraintSetup {angularLimitsMax=<0 50 90>, angularLimitsMin=<0 -50 -80>, angularTarget=<1 0 0>}
      localJointOffset=<15 0 0>
     }]
   }
   n7 = Anim Dynamics - Bone: dyn_bag_02!(ComponentPose=@n6, bAlphaBoolEnabled:bool=True, ExternalForce="0,0,0") -> (Pose) {
    alphaInputType=Bool
    angularDampingOverride=1
    angularSpringConstant=15000
    bAngularSpring=true
    bOverrideAngularDamping=true
    bUseGravityOverride=true
    bUsePlanarLimit=false
    gravityScale=0.35
    simulationSpace=RootRelative
    alphaBoolBlend {blendOption=Linear}
    boundBone {boneName=dyn_bag_02}
    componentAppliedLinearAccClamp=<3000 3000 3000>
    componentLinearAccScale=<0.5 0.5 0.5>
    componentLinearVelScale=<0.5 0.5 0.5>
    gravityOverride {z=-980}
    physicsBodyDefinitions [{
      boundBone {boneName=dyn_bag_02}
      boxExtents=<30 40 45>
      constraintSetup {angularLimitsMax=<0 15 5>, angularLimitsMin=<0 -15 -10>, angularTarget=<1 0 0>}
      localJointOffset=<4 0 0>
     }]
   }
   n8 = Anim Dynamics - Bone: dyn_bag_03!(ComponentPose=@n7, bAlphaBoolEnabled:bool=True, ExternalForce="0,0,0") -> (Pose) {
    alphaInputType=Bool
    angularDampingOverride=1
    angularSpringConstant=15000
    bAngularSpring=true
    bOverrideAngularDamping=true
    bUseGravityOverride=true
    bUsePlanarLimit=false
    gravityScale=0.35
    simulationSpace=RootRelative
    alphaBoolBlend {blendOption=Linear}
    boundBone {boneName=dyn_bag_03}
    componentAppliedLinearAccClamp=<3000 3000 3000>
    componentLinearAccScale=<0.5 0.5 0.5>
    componentLinearVelScale=<0.5 0.5 0.5>
    gravityOverride {z=-980}
    physicsBodyDefinitions [{
      boundBone {boneName=dyn_bag_03}
      boxExtents=<30 40 45>
      constraintSetup {angularLimitsMax=<0 5 5>, angularLimitsMin=<0 -15 -10>, angularTarget=<1 0 0>}
      localJointOffset=<4 0 0>
     }]
   }
   n9 = "Transform (Modify) Bone - Bone: dyn_flap_02_jnt"!(ComponentPose=@n8, bAlphaBoolEnabled:bool=True, Translation="0,0,-0.5") -> (Pose) {
    alphaInputType=Bool
    rotationMode=BMM_Additive
    rotationSpace=BCS_BoneSpace
    scaleMode=BMM_Additive
    scaleSpace=BCS_BoneSpace
    translationMode=BMM_Additive
    translationSpace=BCS_BoneSpace
    boneToModify {boneName=dyn_flap_02_jnt}
    translation {z=-0.5}
   }
   Output Pose(Result=Component To Local(ComponentPose=@n9)) {layerGroup=None}
   n10 = Get Sim Space Settings() -> (Sim Space Settings:Sim Space Settings Structure)
   n11 = Get Override World Gravity() -> (Override World Gravity:Vector)
   AnimGraphNode_SubInput_0 = Input Pose()
   n12 = "Transform (Modify) Bone - Bone: body_root"(ComponentPose=Local To Component(LocalPose=@AnimGraphNode_SubInput_0), Alpha=1, Translation="0,0,0", Rotation="0,0,0", Scale="1,1,1") {
    rotationMode=BMM_Replace
    rotationSpace=BCS_WorldSpace
    scaleSpace=BCS_BoneSpace
    translationMode=BMM_Replace
    translationSpace=BCS_WorldSpace
    alphaBoolBlend {blendOption=Linear}
    alphaScaleBiasClamp {bInterpResult=true, interpSpeedDecreasing=1, interpSpeedIncreasing=1, inRange {max=0, min=-40}, outRange {max=0, min=1}}
    boneToModify {boneName=body_root}
   }
   n13 = "Transform (Modify) Bone - Bone: pelvis"(ComponentPose=@n12, Alpha=1, Translation="0,0,0", Rotation="0,0,0", Scale="1,1,1") {
    rotationMode=BMM_Replace
    rotationSpace=BCS_WorldSpace
    scaleSpace=BCS_BoneSpace
    translationMode=BMM_Replace
    translationSpace=BCS_WorldSpace
    alphaBoolBlend {blendOption=Linear}
    alphaScaleBiasClamp {inRange {max=0, min=-40}, outRange {max=0, min=1}}
    boneToModify {boneName=pelvis}
   }
   n14 = "Transform (Modify) Bone - Bone: thigh_l"(ComponentPose=@n13, Alpha=1, Translation="0,0,0", Rotation="0,0,0", Scale="1,1,1") {
    rotationMode=BMM_Replace
    rotationSpace=BCS_WorldSpace
    scaleSpace=BCS_BoneSpace
    translationMode=BMM_Replace
    translationSpace=BCS_WorldSpace
    alphaBoolBlend {blendOption=Linear}
    alphaScaleBiasClamp {inRange {max=0, min=-40}, outRange {max=0, min=1}}
    boneToModify {boneName=thigh_l}
   }
   n15 = "Transform (Modify) Bone - Bone: thigh_r"(ComponentPose=@n14, Alpha=1, Translation="0,0,0", Rotation="0,0,0", Scale="1,1,1") {
    rotationMode=BMM_Replace
    rotationSpace=BCS_WorldSpace
    scaleSpace=BCS_BoneSpace
    translationMode=BMM_Replace
    translationSpace=BCS_WorldSpace
    alphaBoolBlend {blendOption=Linear}
    alphaScaleBiasClamp {inRange {max=0, min=-40}, outRange {max=0, min=1}}
    boneToModify {boneName=thigh_r}
   }
   n16 = "Transform (Modify) Bone - Bone: calf_l"(ComponentPose=@n15, Alpha=1, Translation="0,0,0", Rotation="0,0,0", Scale="1,1,1") {
    rotationMode=BMM_Replace
    rotationSpace=BCS_WorldSpace
    scaleSpace=BCS_BoneSpace
    translationMode=BMM_Replace
    translationSpace=BCS_WorldSpace
    alphaBoolBlend {blendOption=Linear}
    alphaScaleBiasClamp {inRange {max=0, min=-40}, outRange {max=0, min=1}}
    boneToModify {boneName=calf_l}
   }
   n17 = "Transform (Modify) Bone - Bone: calf_r"(ComponentPose=@n16, Alpha=1, Translation="0,0,0", Rotation="0,0,0", Scale="1,1,1") {
    rotationMode=BMM_Replace
    rotationSpace=BCS_WorldSpace
    scaleSpace=BCS_BoneSpace
    translationMode=BMM_Replace
    translationSpace=BCS_WorldSpace
    alphaBoolBlend {blendOption=Linear}
    alphaScaleBiasClamp {inRange {max=0, min=-40}, outRange {max=0, min=1}}
    boneToModify {boneName=calf_r}
   }
   n18 = "Transform (Modify) Bone - Bone: base_Handle"!(ComponentPose=@n17, Alpha=1, Translation="-1,-2,0", Scale="1,1,0.8") -> (Pose) {
    rotationMode=BMM_Additive
    rotationSpace=BCS_BoneSpace
    scaleMode=BMM_Additive
    scaleSpace=BCS_BoneSpace
    translationMode=BMM_Additive
    translationSpace=BCS_BoneSpace
    alphaBoolBlend {blendOption=Linear}
    boneToModify {boneName=base_Handle}
    scale {z=0.8}
    translation=<-1 -2>
   }
   n19 = "Transform (Modify) Bone - Bone: base_Handle"!(ComponentPose=@n18, Alpha=1, Rotation:Rotator="0,20,0") -> (Pose) {
    rotationMode=BMM_Additive
    rotationSpace=BCS_BoneSpace
    scaleMode=BMM_Additive
    scaleSpace=BCS_BoneSpace
    translationMode=BMM_Additive
    translationSpace=BCS_BoneSpace
    alphaBoolBlend {blendOption=Linear}
    boneToModify {boneName=base_Handle}
    rotation {yaw=20}
   }
   n20 = "Transform (Modify) Bone - Bone: base_bag"!(ComponentPose=@n19, bAlphaBoolEnabled:bool=True, Rotation="0,45,0") -> (Pose) {
    alphaInputType=Bool
    rotationMode=BMM_Additive
    rotationSpace=BCS_BoneSpace
    scaleMode=BMM_Additive
    scaleSpace=BCS_BoneSpace
    translationMode=BMM_Additive
    translationSpace=BCS_BoneSpace
    alphaBoolBlend {blendOption=Linear}
    boneToModify {boneName=base_bag}
    rotation {yaw=45}
   }
   n21 = "Transform (Modify) Bone - Bone: base_Handle"!(ComponentPose=@n20, bAlphaBoolEnabled:bool=True, Translation:Vector="0.5,0,0.3") -> (Pose) {
    alphaInputType=Bool
    rotationMode=BMM_Additive
    rotationSpace=BCS_BoneSpace
    scaleMode=BMM_Additive
    scaleSpace=BCS_BoneSpace
    translationMode=BMM_Additive
    translationSpace=BCS_BoneSpace
    alphaBoolBlend {blendOption=Linear}
    boneToModify {boneName=base_Handle}
    translation {x=0.5, z=0.3}
   }
   n22 = "Transform (Modify) Bone - Bone: base_Handle"(ComponentPose=@n21, bAlphaBoolEnabled=True, Translation="-1,0,0") {
    alphaInputType=Bool
    rotationSpace=BCS_BoneSpace
    scaleSpace=BCS_BoneSpace
    translationMode=BMM_Additive
    translationSpace=BCS_BoneSpace
    alphaBoolBlend {blendOption=Linear}
    boneToModify {boneName=base_Handle}
    translation {x=-1}
   }
   n23 = "Transform (Modify) Bone - Bone: base_Handle"(ComponentPose=@n22, bAlphaBoolEnabled=True, Translation="-0.4,0,0.5") {
    alphaInputType=Bool
    rotationSpace=BCS_BoneSpace
    scaleMode=BMM_Additive
    scaleSpace=BCS_BoneSpace
    translationMode=BMM_Additive
    translationSpace=BCS_BoneSpace
    alphaBoolBlend {blendOption=Linear}
    boneToModify {boneName=base_Handle}
    translation {x=-0.4, z=0.5}
   }
   n24 = RigidBody!(
    ComponentPose=@n23
    bAlphaBoolEnabled:bool
    OverrideWorldGravity:Vector=@n11
    ExternalForce
    SimSpaceSettings:Sim Space Settings Structure=@n10
   ) -> (Pose) {
    alphaInputType=Bool
    bOverrideWorldGravity=true
    overridePhysicsAsset=&/GrowMallet_Axe_Physics
   }
   n25 = RigidBody(ComponentPose=@n24, OverrideWorldGravity=@n11, SimSpaceSettings=@n10) {
    alphaInputType=Bool
    bOverrideWorldGravity=true
    overridePhysicsAsset=&/GrowMallet_Axe_Physics_FE_OffPawn
   }
   n26 = RigidBody(ComponentPose=@n25, OverrideWorldGravity=@n11, SimSpaceSettings=@n10) {
    alphaInputType=Bool
    bOverrideWorldGravity=true
    overridePhysicsAsset=&/GrowMallet_Axe_Physics_FE_OnPawn
   }
   n27 = "Constraint - base_flap (PP)"!(ComponentPose=@n26, Alpha=1, ConstraintWeights_0=0.7876, ConstraintWeights_1:float=0.2124) -> (Pose) {
    boneToModify {boneName=base_flap}
    constraintSetup [{targetBone {boneName=dyn_bag_03}}, {targetBone {boneName=dyn_bag_02}}]
    constraintWeights [1,1]
   }
   n28 = "Constraint - base_brick (PP)"!(ComponentPose=@n27, Alpha=1, ConstraintWeights_0=0.9566, ConstraintWeights_1:float=0.0434) -> (Pose) {
    boneToModify {boneName=base_brick}
    constraintSetup [{targetBone {boneName=dyn_bag_03}}, {targetBone {boneName=dyn_bag_02}}]
    constraintWeights [1,1]
   }
   Save cached pose 'Sim Cache'(Pose=Component To Local(ComponentPose=@n28)) {cacheName=Sim Cache}
  ]}
 EventGraph {nodes [
   Event BlueprintUpdateAnimation()
   > K2Node_Composite_0 = Get Mesh and Body Nodes!(InputPin:Exec) -> (then:Exec) {is=title K2Node_Composite}
   > Set rigidBodyOff(rigidBodyOff:bool=NOT Boolean(A=Get bRigidBodyEnabled())) -> (Output_Get:bool)
   n1 = Get bIsInFrontEnd()
   > Set notFE(notFE:bool=NOT Boolean(A=@n1)) -> (Output_Get:bool)
   n2 = Get bIsOnPawn() -> (bIsOnPawn:bool)
   > Set onPlayerFE(onPlayerFE:bool=AND Boolean(A=@n1, B=@n2)) -> (Output_Get:bool)
   > Set offPlayerFE(offPlayerFE:bool=AND Boolean(A=@n1, B=NOT Boolean(A=@n2))) -> (Output_Get:bool)
   > n3 = Branch(Condition=Get bIsInFrontEnd())
   > Set Sim Space Settings(Sim Space Settings:Sim Space Settings Structure=Get Sim Space Settings_FrontEnd() -> (Sim Space Settings_FrontEnd:Sim Space Settings Structure)) -> (Output_Get:Sim Space Settings Structure)
   > Set Override World Gravity(Override World Gravity:Vector="0,0,-1660") -> (Output_Get:Vector)
   Set Sim Space Settings(execute=@n3.else, Sim Space Settings=Get Sim Space Settings_PIE() -> (Sim Space Settings_PIE:Sim Space Settings Structure))
   > Set Override World Gravity(Override World Gravity="0,0,-980")
   n4 = Event OnCharacterCustomizationLoaded()
   n5 = Get Array Item with EFort Custom Part Type(Array:Array of Custom Character Part Object References=@n4.CustomParts) -> (Output:Custom Character Part Object Reference)
   n6 = Get GenderPermitted!(self:Custom Character Part Object Reference=@n5) -> (GenderPermitted:EFortCustomGender Enum=Invalid)
   > Set isMale(isMale:bool="Equal (Enum)"(A:EFortCustomGender Enum=@n6, B:EFortCustomGender Enum=Male)) -> (Output_Get:bool)
   > Set isFemale(isFemale:bool="Equal (Enum)"(A=@n6, B=Female)) -> (Output_Get:bool)
   > Set isLarge(isLarge:bool="Equal (Enum)"(
     A:EFortCustomBodyType Enum=Get BodyTypesPermitted!(self:Custom Character Part Object Reference=@n5) -> (BodyTypesPermitted:EFortCustomBodyType Enum=NONE)
     B:EFortCustomBodyType Enum=Large
    )) -> (Output_Get:bool)
   > Set isMedium(isMedium:bool=NOT Boolean(A=Get isLarge() -> (isLarge:bool))) -> (Output_Get:bool)
   n7 = Get isMale() -> (isMale:bool)
   n8 = Get isLarge()
   > Set isMaleL(isMaleL:bool=AND Boolean(A=@n7, B=@n8)) -> (Output_Get:bool)
   > Set isMaleM(isMaleM:bool=AND Boolean(A=@n7, B=Get isMedium() -> (isMedium:bool))) -> (Output_Get:bool)
   > Set isFemaleL(isFemaleL:bool=AND Boolean(A=@n8, B=Get isFemale() -> (isFemale:bool))) -> (Output_Get:bool)
   TryGetPawnOwner()
  ]}
 Get Mesh and Body Nodes {
  path=EventGraph.K2Node_Composite_0.Get Mesh and Body Nodes
  nodes [
   Inputs!() -> (InputPin:Exec)
   n1 = GetPartSkeletalMeshComponent()
   > Is Valid(InputObject=@n1)
   n2 = GetAttachParent!(self:Scene Component Object Reference=@n1) -> (ReturnValue:Scene Component Object Reference)
   > Is Valid(InputObject=@n2)
   > Set Body_Mesh(Body_Mesh:Scene Component Object Reference=@n2) -> (Output_Get:Scene Component Object Reference)
   > n3 = Sequence!(execute) -> (then_0, then_1, then_2:Exec, then_3:Exec, then_4:Exec, then_5:Exec, then_6:Exec)
   n4 = Get Body_Mesh() -> (Body_Mesh:Scene Component Object Reference)
   n5 = BreakTransform(InTransform=GetSocketTransform(self=@n4, InSocketName=root, TransformSpace=RTS_World))
   Set root_location(execute=@n3.then_6, root_location:Vector=@n5) -> (Output_Get:Vector)
   > Set root_space_rotation(root_space_rotation:Rotator=@n5.Rotation) -> (Output_Get:Rotator)
   > Outputs!(then:Exec)
   n6 = BreakTransform(InTransform=GetSocketTransform(self=@n4, InSocketName=pelvis, TransformSpace=RTS_World))
   Set pelvis_location(execute=@n3.then_0, pelvis_location:Vector=@n6) -> (Output_Get:Vector)
   > Set pelvis_rotation(pelvis_rotation:Rotator=@n6.Rotation) -> (Output_Get:Rotator)
   n7 = BreakTransform(InTransform=GetSocketTransform(self=@n4, InSocketName=thigh_l, TransformSpace=RTS_World))
   Set thigh_left_location(execute=@n3.then_1, thigh_left_location:Vector=@n7) -> (Output_Get:Vector)
   > Set thigh_left_rotation(thigh_left_rotation:Rotator=@n7.Rotation) -> (Output_Get:Rotator)
   n8 = BreakTransform(InTransform=GetSocketTransform(self=@n4, InSocketName=thigh_r, TransformSpace=RTS_World))
   Set thigh_right_location(execute=@n3.then_2, thigh_right_location:Vector=@n8) -> (Output_Get:Vector)
   > Set thigh_right_rotation(thigh_right_rotation:Rotator=@n8.Rotation) -> (Output_Get:Rotator)
   n9 = BreakTransform(InTransform=GetSocketTransform(self=@n4, InSocketName=calf_r, TransformSpace=RTS_World))
   Set calf_right_location(execute=@n3.then_4, calf_right_location:Vector=@n9) -> (Output_Get:Vector)
   > Set calf_right_rotation(calf_right_rotation:Rotator=@n9.Rotation) -> (Output_Get:Rotator)
   n10 = BreakTransform(InTransform=GetSocketTransform(self=@n4, InSocketName=calf_l, TransformSpace=RTS_World))
   Set calf_left_location(execute=@n3.then_3, calf_left_location:Vector=@n10) -> (Output_Get:Vector)
   > Set calf_left_rotation(calf_left_rotation:Rotator=@n10.Rotation) -> (Output_Get:Rotator)
   n11 = BreakTransform(InTransform=GetSocketTransform(self=@n4, InSocketName=root, TransformSpace=RTS_World))
   Set body_root_location(execute=@n3.then_5, body_root_location:Vector=@n11) -> (Output_Get:Vector)
   > Set body_root_rotation(body_root_rotation:Rotator=@n11.Rotation) -> (Output_Get:Rotator)
  ]
 }
]
variables [
 Body_Mesh {type=SceneComponent, default=null}
 body_root_location {type=Vector}
 body_root_rotation {type=Rotator}
 calf_left_location {type=Vector}
 calf_left_rotation {type=Rotator}
 calf_right_location {type=Vector}
 calf_right_rotation {type=Rotator}
 isFemale {type=boolean}
 isFemaleL {type=boolean}
 isLarge {type=boolean}
 isMale {type=boolean}
 isMaleL {type=boolean}
 isMaleM {type=boolean}
 isMedium {type=boolean}
 notFE {type=boolean}
 offPlayerFE {type=boolean}
 onPlayerFE {type=boolean}
 Override World Gravity {type=Vector}
 pelvis_location {type=Vector}
 pelvis_rotation {type=Rotator}
 rigidBodyOff {type=boolean}
 root_location {type=Vector}
 root_space_rotation {type=Rotator}
 Sim Space Settings {
  type=SimSpaceSettings
  default {
   dampingAlpha=1
   maxAngularAcceleration=10000
   maxAngularVelocity=10000
   maxLinearAcceleration=10000
   maxLinearVelocity=10000
   velocityScaleZ=1
   worldAlpha=0
   externalLinearDragV=<0 0 0>
  }
 }
 Sim Space Settings_FrontEnd {
  type=SimSpaceSettings
  default {
   dampingAlpha=0
   maxAngularAcceleration=750
   maxAngularVelocity=150
   maxLinearAcceleration=1500
   maxLinearVelocity=1000
   velocityScaleZ=0.25
   worldAlpha=1
   externalLinearDragV=<0.2 0.2 0.2>
  }
 }
 Sim Space Settings_PIE {
  type=SimSpaceSettings
  default {
   dampingAlpha=0
   maxAngularAcceleration=30
   maxAngularVelocity=8
   maxLinearAcceleration=2400
   maxLinearVelocity=700
   velocityScaleZ=0.7
   worldAlpha=1
   externalLinearDragV=<0.5 0.5 0.5>
  }
 }
 thigh_left_location {type=Vector}
 thigh_left_rotation {type=Rotator}
 thigh_right_location {type=Vector}
 thigh_right_rotation {type=Rotator}
]
